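# NS_Beat is a Niagara vehicle boost cosmetic that trails red and blue ribbons and meshes behind the car, with bursts where the meshes die.
# A mesh emitter launches particles backward from partitioned skeletal sockets and records their death positions and colors in system arrays through scratch modules.
# Sprite and mesh emitters spawn from those arrays as glows and tumbling half-cube debris, mostly choosing red or blue by death-position sign.
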NiagaraSystem /VehicleCosmetics/Boosts/Boost_Beat/FX/NS_Beat
dependencies {
 usedDataInterfaces [
  &NiagaraDataInterfaceArrayFloat
  &NiagaraDataInterfaceArrayFloat3
  &NiagaraDataInterfaceVectorCurve
  &NiagaraDataInterfaceCurve
  &NiagaraDataInterfaceMeshRendererInfo
  &NiagaraDataInterfaceSpriteRendererInfo
  &NiagaraDataInterfaceSkeletalMesh
  &NiagaraDataInterfaceArrayPosition
  &NiagaraDataInterfaceArrayColor
 ]
 usedDynamicInputs [/Niagara/DynamicInputs/Math/Floor_Float, &Divide_Float]
 usedModules [
  &self:SystemSpawnScript.SystemScriptSource.SystemScriptGraph.NiagaraNodeAssignment_0.SetVariables_#1
  &SystemState
  &self:ReadDeathArrayScratch
 ]
}
emitters {
 NE_VC_Boost_Ribbon_Parent {
  emitterUpdateScript [
   &EmitterState()
   &VC_FilteredSocketCount()
   &VC_RateQualityHandler(High Spawn Rate=60, Low Spawn Rate=60, Medium Spawn Rate=60, Use Custom Rate?=true)
   &VC_SpawnRibbon_v1(Number of Ribbons=$Emitter.PortCount, SpawnRate=$Emitter.Spawn Rate)
  ]
  particleSpawnScript [
   &VC_PaintedVariantsOverride(DefaultColor=<rgba 0.3739 0.0593 0.0565 1>, R1 DefaultColor=<rgba 0 0.3467 1 1>, Use R1 Painted Overrides=true)
   &VC_InitializeRibbon_v1(Color=$Particles.Color, Enable Skeletal Position=true, Lifetime=0.15, Ribbon Width=7)
   &AddVelocity(Velocity=<-9500 0 0>) {module=AddVelocity001}
   /SetVariables() {isSetParameters=true}
  ]
  particleUpdateScript [
   /SetVariables(
    Particles.BlueOutlineWidth:NiagaraFloat=&SelectFloatFromArray(Direct Array Index=&SpawnGroup(), Float Selection Array=DataInterface() {propertyValues {FloatData [-11,0]}})
    Particles.BlueWidth:NiagaraFloat=&SelectFloatFromArray(Direct Array Index=&SpawnGroup(), Float Selection Array=DataInterface() {propertyValues {FloatData [-9,0]}})
    Particles.RedOutlineWidth:NiagaraFloat=&SelectFloatFromArray(Direct Array Index=&SpawnGroup(), Float Selection Array=DataInterface() {propertyValues {FloatData [0,11]}})
    Particles.RedWidth:NiagaraFloat=&SelectFloatFromArray(Direct Array Index=&SpawnGroup(), Float Selection Array=DataInterface() {propertyValues {FloatData [0,9]}})
   ) {isSetParameters=true}
   &ParticleState()
   &InheritSourceMovement(ApplyForce=false, ApplyPosition=true, ApplyPositionFalloff=false)
   &SolveForcesAndVelocity(Clamp Velocity=false, Limit Acceleration=false)
   &VC_UpdateRibbon_v1(Enable Skeletal Position=true)
  ]
  rendererClasses [&NiagaraRibbonRendererProperties]
  renderers [
   {rendererClass=&NiagaraRibbonRendererProperties}
   {rendererClass=&NiagaraRibbonRendererProperties, rendererIndex=1}
   {rendererClass=&NiagaraRibbonRendererProperties, rendererIndex=2}
   {rendererClass=&NiagaraRibbonRendererProperties, rendererIndex=3}
  ]
 }
 NE_VC_Boost_Sprite_Parent {
  emitterUpdateScript [
   &EmitterState()
   &VC_FilteredSocketCount()
   &SpawnRate(SpawnRate=2, Use Spawn Probability=true) {enabled=false}
   &self:SpawnDeathArrayScratch(ClearDeathPositions=false) {module=SpawnDeathArrayScratch}
   &self:SpawnDeathArrayScratch(ClearDeathPositions=false, CountPerPosition=14, Spawn Group=1) {module=SpawnDeathArrayScratch001}
  ]
  particleSpawnScript [
   /SetVariables(Particles.VisibilityTag=$Emitter.SpawnGroup) {isSetParameters=true, module=SetVariables_#2}
   &VC_PaintedVariantsOverride(DefaultColor=<rgba 2.6823 2.8957 5 1>, R1 DefaultColor=<rgba 5 2.605 2.605 1>, Use R1 Painted Overrides=true) {enabled=false}
   &InitializeParticle(
    Color Mode=::NewEnumerator0
    Lifetime=&MergeHelper_FloatSwitch(
     0=&RandomRangeFloat(Minimum=0.2, Maximum=0.3)
     1=&RandomRangeFloat(Minimum=0.3, Maximum=0.15)
     2=&RandomRangeFloat(Minimum=0.2, Maximum=0.3)
     3=&RandomRangeFloat(Minimum=0.2, Maximum=0.3)
     4=&RandomRangeFloat(Minimum=1, Maximum=2)
     Index=$Emitter.SpawnGroup
     Num=::NewEnumerator0
    )
    Mesh Scale Mode=::NewEnumerator3
    Mesh Uniform Scale=&MergeHelper_FloatSwitch(0=0, 1=7.5, 2=8, 3=0, Index=$Emitter.SpawnGroup, Num=::NewEnumerator2)
    Position Offset:NiagaraPosition=$System.ReadDeathArrayScratch.DeathPosition2
    Position Offset Coordinate Space=::Local
    Sprite Size=&MergeHelper_Vector2Switch(
     0=&Vector2DFromFloat(Value=&RandomRangeFloat(Minimum=300, Maximum=400))
     1=&RandomRangeVector2D(Minimum=<1 48>, Maximum=<3 60>)
     2=?
     3=?
     4=?
     Index=$Emitter.SpawnGroup
     Num=::NewEnumerator0
    )
    Sprite Size Mode=::NewEnumerator4
    UsePositionOffset=true
    Write Material Random=true
   )
   &PartitionParticles(Number Of Partitions=$Emitter.PortCount) {enabled=false}
   &SkeletalMeshLocation(
    Bone / Socket Index=$Particles.PartitionParticles.Partition
    Bone / Socket Sampling Mode=::NewEnumerator5
    MaskSpawnGroup=false
    Mesh Sampling Type=::NewEnumerator6
    Sampled Position Offset=&MergeHelper_VectorSwitch(
     0=Linked() {linkedVariable {name=System.ReadDeathArrayScratch.DeathPosition2, type=NiagaraPosition}}
     1=<-850 0 0>
     2=<0 0 0>
     3=<0 0 0>
     4=<0 0 0>
     Index=$Emitter.SpawnGroup
     Num=::NewEnumerator3
    )
    Skeletal Mesh=$System.SkeletalMesh
   ) {enabled=false}
   &ShapeLocation(MaskSpawnGroup=true, Offset Mode=::NewEnumerator0, Override Seed=false, Spawn Group Mask=4, UseOwnerScale=true) {module=ShapeLocation001}
   &AddVelocity(
    Module.Use Falloff Distance=false
    Origin Offset:NiagaraPosition=$Particles.ShapeLocation001.ShapePosition
    Override Seed=false
    Velocity Mode=::NewEnumerator1
    Velocity Speed=&MergeHelper_FloatSwitch(
     0=0
     1=&RandomRangeFloat(Minimum=350, Maximum=600)
     2=?
     3=?
     Index=$Emitter.SpawnGroup
     Num=::NewEnumerator0
    )
   )
   &Color(Color=&MergeHelper_ColorSwitch(
     0=&MakeCustomLinearColorFromBool(
      Bool=&SetBoolByFloatComparison(
       A=&MakeFloatFromVector(Vector=Linked() {linkedVariable {name=System.ReadDeathArrayScratch.DeathPosition2, type=NiagaraPosition}}, Channel=::NewEnumerator1)
       B=0
      )
      True Color=<rgba 1.1979 2.855 10 0.0522>
      False Color=<rgba 10 0.6771 0.6771 0.0348>
     )
     1=&MakeCustomLinearColorFromBool(
      Bool=&SetBoolByFloatComparison(
       A=&MakeFloatFromVector(Vector=Linked() {linkedVariable {name=System.ReadDeathArrayScratch.DeathPosition2, type=NiagaraPosition}}, Channel=::NewEnumerator1)
       B=0
      )
      True Color=<rgba 1.1979 2.855 10 0.15>
      False Color=<rgba 10 0.6771 0.6771 0.15>
     )
     2=&MakeCustomLinearColorFromBool(
      Bool=&SetBoolByFloatComparison(
       A=&MakeFloatFromVector(Vector=Linked() {linkedVariable {name=System.ReadDeathArrayScratch.DeathPosition2, type=NiagaraPosition}}, Channel=::NewEnumerator1)
       B=0
      )
      True Color=<rgba 1.1979 2.855 10 0.4522>
      False Color=<rgba 10 2.5521 2.5521 0.4087>
     )
     3=&MakeCustomLinearColorFromBool(
      Bool=&SetBoolByFloatComparison(
       A=&MakeFloatFromVector(Vector=Linked() {linkedVariable {name=System.ReadDeathArrayScratch.DeathPosition2, type=NiagaraPosition}}, Channel=::NewEnumerator1)
       B=0
      )
      True Color=<rgba 1.1979 2.855 10 0.4522>
      False Color=<rgba 10 2.5521 2.5521 0.4087>
     )
     4=&MakeCustomLinearColorFromBool(
      Bool=&SetBoolByFloatComparison(
       A=&MakeFloatFromVector(Vector=Linked() {linkedVariable {name=System.ReadDeathArrayScratch.DeathPosition2, type=NiagaraPosition}}, Channel=::NewEnumerator1)
       B=0
      )
      True Color=<rgba 1.1979 2.855 10 0.4522>
      False Color=<rgba 10 2.5521 2.5521 0.4087>
     )
     Index=$Emitter.SpawnGroup
     Num=::NewEnumerator0
    ))
  ]
  particleUpdateScript [
   &ParticleState()
   &ScaleSpriteSize(
    Scale Sprite Size Mode=::NewEnumerator0
    Uniform Scale Factor=&MergeHelper_FloatSwitch(
     0=&FloatFromCurve(FloatCurve=DataInterface() {propertyValues {Curve {keys [<at 0 1 RCIM_Cubic>,<at 1 0 RCIM_Cubic>]}}})
     1=1
     2=?
     3=&FloatFromCurve(FloatCurve=DataInterface() {propertyValues {Curve {keys [<at 0 1 RCIM_Cubic>,<at 1 0 RCIM_Cubic>]}}})
     Index=&SpawnGroup()
     Num=::NewEnumerator0
    )
   )
   &ScaleColor(
    Scale Alpha=&FloatFromCurve(FloatCurve=DataInterface() {propertyValues {Curve {
        defaultValue=1
        keys [
         {arriveTangent=4.9594, interpMode=RCIM_Cubic, leaveTangent=4.9594, tangentMode=RCTM_User, time=0, value=0}
         {interpMode=RCIM_Cubic, tangentMode=RCTM_User, time=0.5, value=1}
         {tangentMode=RCTM_User, time=1, value=0}
        ]
       }}})
    Scale RGB=&VectorFromFloat(Value=&MergeHelper_FloatSwitch(0=10, 1=150, 2=?, 3=?, Index=&SpawnGroup(), Num=::NewEnumerator0))
    ScaleA=true
    ScaleRGB=true
   )
   &SubUVAnimation(
    Evaluation Type=::NewEnumerator0
    Output Alternate Renderer Binding=true
    "Sprite Renderer "=DataInterface() {propertyValues {SpriteRenderer=&self:NE_VC_Boost_Sprite_Parent_0.NiagaraSpriteRendererProperties_0}}
    SubUV Animation Mode:ENiagara_SubUVLookupModeV2=::NewEnumerator2
    UseEndFrame=false
    UseStartFrame=false
   ) {module=SubUVAnimation002}
   &GravityForce(Gravity=&RandomRangeVector(Minimum=<0 0 -600>, Maximum=<0 0 -300>))
   &SolveForcesAndVelocity(Clamp Velocity=false, Limit Acceleration=false)
  ]
  rendererClasses [&NiagaraSpriteRendererProperties]
  renderers [
   {rendererClass=&NiagaraSpriteRendererProperties}
   {rendererClass=&NiagaraSpriteRendererProperties, rendererIndex=1}
  ]
 }
 NE_VC_Boost_Sprite_Parent007 {
  emitterSpawnScript [/SetVariables() {isSetParameters=true}]
  emitterUpdateScript [
   &EmitterState(Life Cycle Mode=::NewEnumerator1, UseLoopDelay=false)
   &VC_FilteredSocketCount()
   &SpawnRate(Spawn Probability=0.5, SpawnGroup=1, SpawnRate=5, Use Spawn Probability=false)
   &SpawnRate(Spawn Probability=0.5, SpawnRate=4, Use Spawn Probability=false) {module=SpawnRate001}
   /SetVariables() {enabled=false, isSetParameters=true}
   &self:ScratchModule_04(
    DeathColorArray:NiagaraDataInterfaceArrayColor=$System.DeathColorArray
    DeathPositionsArray:NiagaraDataInterfaceArrayPosition=$System.DeathArray
   ) {module=ScratchModule_04}
  ]
  particleSpawnScript [
   /SetVariables(Particles.VisibilityTag=$Emitter.SpawnGroup) {isSetParameters=true}
   &VC_PaintedVariantsOverride(DefaultColor=<rgba 0 0 1 1>, R1 DefaultColor=<rgba 5 2.605 2.605 1>, Use R1 Painted Overrides=true) {enabled=false}
   &InitializeParticle(
    Color=$Particles.Color
    Lifetime Max=1
    Lifetime Min=0.75
    Lifetime Mode=::NewEnumerator1
    Mesh Scale Mode=::NewEnumerator3
    Mesh Uniform Scale=2
    Sprite Rotation Mode=::NewEnumerator2
    UsePositionOffset=false
    Write Material Random=true
   )
   &PartitionParticles(Number Of Partitions=$Emitter.PortCount)
   &SkeletalMeshLocation(
    Bone / Socket Index=$Particles.PartitionParticles.Partition
    Bone / Socket Sampling Mode=::NewEnumerator5
    MaskSpawnGroup=false
    Mesh Sampling Type=::NewEnumerator6
    Sampled Position Offset=&AddVectorToPosition(Position=<0 0 25>, Vector=<0 0 0>)
    Skeletal Mesh=$System.SkeletalMesh
   )
   &AddVelocity(Velocity=&MergeHelper_VectorSwitch(0=<-1000 0 0>, 1=<-950 0 0>, 2=?, 3=?, Index=&SpawnGroup(), Num=::NewEnumerator0))
   &InitialMeshOrientation(
    Mesh Axis To Orient=&MergeHelper_VectorSwitch(
     0=&RandomRangeVector(Minimum=<0 0 0>, Maximum=<0 -1 0>)
     1=<0 0 0>
     2=?
     3=?
     Index=&SpawnGroup()
     Num=::NewEnumerator0
    )
    Mesh Orientation Mode=::NewEnumerator2
    Orientation Vector=&MergeHelper_VectorSwitch(0=<0 0 0>, 1=<0 1 0>, 2=?, 3=?, Index=&SpawnGroup(), Num=::NewEnumerator0)
    Rotation=&SelectVectorFromArray(
     Array Sampling Mode=::NewEnumerator0
     Direct Array Index=?
     Vector Selection Array=DataInterface() {propertyValues {FloatData [<0 0 0>, <0 0.25 0>, <0 0.5 0>, <0 0.75 0>]}}
     Random Seed=0
    )
    Use Rotation Vector=true
   )
  ]
  particleUpdateScript [
   &ParticleState()
   &ScaleMeshSize(Scale Factor=&VectorFromCurve(VectorCurve=DataInterface() {propertyValues {
       XCurve {
        defaultValue=1
        keys [{interpMode=RCIM_Cubic, tangentMode=RCTM_User, time=0, value=0}, {interpMode=RCIM_Cubic, tangentMode=RCTM_User, time=0.25, value=1}]
       }
       YCurve {
        defaultValue=1
        keys [{interpMode=RCIM_Cubic, tangentMode=RCTM_User, time=0, value=0}, {interpMode=RCIM_Cubic, tangentMode=RCTM_User, time=0.25, value=1}]
       }
       ZCurve {
        defaultValue=1
        keys [{interpMode=RCIM_Cubic, tangentMode=RCTM_User, time=0, value=0}, {interpMode=RCIM_Cubic, tangentMode=RCTM_User, time=0.25, value=1}]
       }
      }}))
   &Color(Color=&MaskLinearColor(
     Mask Source=1
     Mask Parameter=$Particles.Initial.PartitionParticles.Partition
     Mask=0
     Masked Linear Color=<rgba 0.0024 0.0324 0.1615 1>
     Passthrough Linear Color=<rgba 0.1406 0 0 1>
    ))
   &SolveForcesAndVelocity(Clamp Velocity=false, Limit Acceleration=false)
   &DynamicMaterialParameters(
    Index 0 Param 1=1
    Index 0 Param 2=1
    Module.Index0Float0Write=true
    Module.Index0Float1Write=true
    Module.Index0Float2Write=true
    Module.Index0Float3Write=true
   )
   &self:WriteDeathArrayScratch() {module=WriteDeathArrayScratch}
   /SetVariables(Particles.ArrowsColor:LinearColor=&MultiplyLinearColor(
     Linear Color=<rgba 1 1 1 1>
     Scale Factor=&MakeVector4FromVectorAndFloat("Vector (RGB)"=<1 1 1>, "Float (Alpha)"=1)
    )) {isSetParameters=true}
  ]
  rendererClasses [&NiagaraMeshRendererProperties, &NiagaraSpriteRendererProperties]
  renderers [
   {rendererClass=&NiagaraMeshRendererProperties, rendererIndex=0}
   {rendererClass=&NiagaraMeshRendererProperties, rendererIndex=1}
   {rendererClass=&NiagaraSpriteRendererProperties, rendererIndex=2}
  ]
 }
 NE_VC_Boost_Sprite_Parent010 {
  emitterUpdateScript [
   &EmitterState()
   &VC_FilteredSocketCount()
   &self:SpawnDeathArrayScratch(ClearDeathPositions=false) {module=SpawnDeathArrayScratch}
   &self:SpawnDeathArrayScratch(ClearDeathPositions=false, Spawn Group=1) {module=SpawnDeathArrayScratch001}
   &self:SpawnDeathArrayScratch(Spawn Group=2) {module=SpawnDeathArrayScratch002}
   &SpawnRate(SpawnRate=2, Use Spawn Probability=true) {enabled=false}
  ]
  particleSpawnScript [
   /SetVariables(Particles.VisibilityTag=$Emitter.SpawnGroup) {isSetParameters=true, module=SetVariables_#2}
   &VC_PaintedVariantsOverride(DefaultColor=<rgba 2.6823 2.8957 5 1>, R1 DefaultColor=<rgba 5 2.605 2.605 1>, Use R1 Painted Overrides=true) {enabled=false}
   &InitializeParticle(
    Color=<rgba 1 1 1 0>
    Lifetime=&MergeHelper_FloatSwitch(
     0=&RandomRangeFloat(Minimum=0.2, Maximum=0.3)
     1=&RandomRangeFloat(Minimum=0.1, Maximum=0.15)
     2=&RandomRangeFloat(Minimum=0.15, Maximum=0.2)
     3=&RandomRangeFloat(Minimum=0.2, Maximum=0.3)
     4=&RandomRangeFloat(Minimum=1, Maximum=2)
     Index=$Emitter.SpawnGroup
     Num=::NewEnumerator1
    )
    Mesh Scale Mode=::NewEnumerator3
    Mesh Uniform Scale=&MergeHelper_FloatSwitch(0=2, 1=7.5, 2=8, 3=?, Index=$Emitter.SpawnGroup, Num=::NewEnumerator1)
    Position Offset:NiagaraPosition=$System.ReadDeathArrayScratch.DeathPosition2
    Position Offset Coordinate Space=::Local
    Sprite Size Mode=::NewEnumerator3
    Uniform Sprite Size=0
    UsePositionOffset=true
    Write Material Random=true
   )
   &PartitionParticles(Number Of Partitions=$Emitter.PortCount)
   &SkeletalMeshLocation(
    Bone / Socket Index=$Particles.PartitionParticles.Partition
    Bone / Socket Sampling Mode=::NewEnumerator5
    MaskSpawnGroup=false
    Mesh Sampling Type=::NewEnumerator6
    Sampled Position Offset=&MergeHelper_VectorSwitch(
     0=Linked() {linkedVariable {name=System.ReadDeathArrayScratch.DeathPosition2, type=NiagaraPosition}}
     1=<-850 0 0>
     2=<0 0 0>
     3=<0 0 0>
     4=<0 0 0>
     Index=$Emitter.SpawnGroup
     Num=::NewEnumerator3
    )
    Skeletal Mesh=$System.SkeletalMesh
   ) {enabled=false}
   &AddVelocity(
    Cone Angle=150
    Module.UseConeFalloff=false
    Module.UseVelDistribution=false
    Override Seed=false
    Velocity Mode=::NewEnumerator2
    Velocity Speed=&MergeHelper_FloatSwitch(
     0=&RandomRangeFloat(Minimum=500, Maximum=700)
     1=0
     2=0
     3=0
     4=0
     Index=$Emitter.SpawnGroup
     Num=::NewEnumerator3
    )
   )
   &ShapeLocation(
    MaskSpawnGroup=true
    Offset Mode=::NewEnumerator0
    Override Seed=false
    Rotation Mode=::NewEnumerator5
    Spawn Group Mask=4
    UseOwnerScale=true
   ) {module=ShapeLocation001}
   &InitialMeshOrientation(
    Mesh Axis To Orient=&MergeHelper_VectorSwitch(
     0=<1 1 1>
     1=&RandomRangeVector(Minimum=<-0.2 -0.2 0.5>, Maximum=<0.2 0.2 1>)
     2=&RandomRangeVector(Minimum=<-0.2 -0.2 0.5>, Maximum=<0.2 0.2 1>)
     3=?
     Index=$Emitter.SpawnGroup
     Num=::NewEnumerator1
    )
    Mesh Orientation Mode=::NewEnumerator2
    Orientation Vector=&MergeHelper_VectorSwitch(0=<1 0 0>, 1=<1 0 0>, 2=<1 0 0>, 3=?, Index=$Emitter.SpawnGroup, Num=::NewEnumerator1)
    Use Rotation Vector=true
   )
   &Color(Color=&MergeHelper_ColorSwitch(
     0=$System.ReadDeathArrayScratch.DeathColorArray
     1=&MakeCustomLinearColorFromBool(
      Bool=&SetBoolByFloatComparison(
       A=&MakeFloatFromVector(Vector=Linked() {linkedVariable {name=System.ReadDeathArrayScratch.DeathPosition2, type=NiagaraPosition}}, Channel=::NewEnumerator1)
       B=0
      )
      True Color=<rgba 5.9896 14.2748 50 0.4522>
      False Color=<rgba 50 3.3854 3.3854 0.8696>
     )
     2=&MakeCustomLinearColorFromBool(
      Bool=&SetBoolByFloatComparison(
       A=&MakeFloatFromVector(Vector=Linked() {linkedVariable {name=System.ReadDeathArrayScratch.DeathPosition2, type=NiagaraPosition}}, Channel=::NewEnumerator1)
       B=0
      )
      True Color=<rgba 1.1979 2.855 10 0.4522>
      False Color=<rgba 10 2.5521 2.5521 0.4087>
     )
     3=&MakeCustomLinearColorFromBool(
      Bool=&SetBoolByFloatComparison(
       A=&MakeFloatFromVector(Vector=Linked() {linkedVariable {name=System.ReadDeathArrayScratch.DeathPosition2, type=NiagaraPosition}}, Channel=::NewEnumerator1)
       B=0
      )
      True Color=<rgba 1.1979 2.855 10 0.4522>
      False Color=<rgba 10 2.5521 2.5521 0.4087>
     )
     4=&MakeCustomLinearColorFromBool(
      Bool=&SetBoolByFloatComparison(
       A=&MakeFloatFromVector(Vector=Linked() {linkedVariable {name=System.ReadDeathArrayScratch.DeathPosition2, type=NiagaraPosition}}, Channel=::NewEnumerator1)
       B=0
      )
      True Color=<rgba 1.1979 2.855 10 0.4522>
      False Color=<rgba 10 2.5521 2.5521 0.4087>
     )
     Index=$Emitter.SpawnGroup
     Num=::NewEnumerator1
    ))
  ]
  particleUpdateScript [
   &ParticleState()
   &ScaleSpriteSize(
    Scale Sprite Size Mode=::NewEnumerator0
    Uniform Scale Factor=&MergeHelper_FloatSwitch(
     0=0
     1=0
     2=0
     3=&FloatFromCurve(FloatCurve=DataInterface() {propertyValues {Curve {keys [<at 0 1 RCIM_Cubic>,<at 1 0 RCIM_Cubic>]}}})
     4=0
     Index=0
     Num=::NewEnumerator3
    )
   )
   &UpdateMeshOrientation(
    Rotation Rate=&MergeHelper_FloatSwitch(
     0=&RandomRangeFloat(Minimum=1, Maximum=3)
     1=&RandomRangeFloat(Minimum=0.75, Maximum=3)
     2=&RandomRangeFloat(Minimum=0.75, Maximum=3)
     3=?
     Index=&SpawnGroup()
     Num=::NewEnumerator1
    )
    Rotation Vector=&MergeHelper_VectorSwitch(
     0=&RandomRangeVector(Minimum=<-1 -1 -1>, Maximum=<1 1 1>)
     1=&RandomRangeVector(Minimum=<0 0 -0.3>, Maximum=<0 0 0.6>)
     2=&RandomRangeVector(Minimum=<0 0 -0.3>, Maximum=<0 0 0.6>)
     3=&RandomRangeVector(Minimum=<-1 -1 -1>, Maximum=<1 1 1>)
     Index=&SpawnGroup()
     Num=::NewEnumerator1
    )
   )
   &ScaleMeshSize(Scale Factor=&VectorFromCurve(VectorCurve=DataInterface() {propertyValues {
       XCurve {
        defaultValue=1
        keys [{interpMode=RCIM_Cubic, tangentMode=RCTM_User, time=0, value=0}, {interpMode=RCIM_Cubic, tangentMode=RCTM_User, time=0.25, value=1}]
       }
       YCurve {
        defaultValue=1
        keys [{interpMode=RCIM_Cubic, tangentMode=RCTM_User, time=0, value=0}, {interpMode=RCIM_Cubic, tangentMode=RCTM_User, time=0.25, value=1}]
       }
       ZCurve {
        defaultValue=1
        keys [{interpMode=RCIM_Cubic, tangentMode=RCTM_User, time=0, value=0}, {interpMode=RCIM_Cubic, tangentMode=RCTM_User, time=0.25, value=1}]
       }
      }}))
   &GravityForce(Gravity=&MergeHelper_VectorSwitch(0=<0 0 -980>, 1=<0 0 0>, 2=<0 0 0>, 3=<0 0 0>, 4=<0 0 0>, Index=&SpawnGroup(), Num=::NewEnumerator3))
   &ScaleColor(
    Scale Alpha=&FloatFromCurve(FloatCurve=DataInterface() {propertyValues {Curve {
        defaultValue=1
        keys [
         {arriveTangent=4.9594, interpMode=RCIM_Cubic, leaveTangent=4.9594, tangentMode=RCTM_User, time=0, value=0}
         {interpMode=RCIM_Cubic, tangentMode=RCTM_User, time=0.5, value=1}
         {tangentMode=RCTM_User, time=1, value=0}
        ]
       }}})
    Scale RGB=&VectorFromFloat(Value=1)
    ScaleA=true
    ScaleRGB=true
   )
   &DynamicMaterialParameters(
    Index 0 Param 1=1
    Index 0 Param 2=1
    Index 3 Param 4=&FloatFromCurve(FloatCurve=DataInterface() {propertyValues {Curve {keys [<at 0 0>,<at 1 3>]}}})
    Module.Index0Float0Write=true
    Module.Index0Float1Write=true
    Module.Index0Float2Write=true
    Module.Index0Float3Write=true
    Module.Index3Float0Write=true
    Module.Index3Float1Write=true
    Module.Index3Float2Write=true
    Module.Index3Float3Write=true
    Write Parameter Index 3=::NewEnumerator1
   )
   /SetVariables(Particles.HalfCubeColorOutline:LinearColor=&MultiplyLinearColor(
     Linear Color=&HueShiftLinearColor(Hue=0, Saturation=0.344, Value=1500, Linear Color=$Particles.Color)
     Scale Factor=&MakeVector4FromVectorAndFloat("Vector (RGB)"=<1 1 1>, "Float (Alpha)"=1)
    )) {isSetParameters=true}
   &SolveForcesAndVelocity(Clamp Velocity=false, Limit Acceleration=false)
  ]
  rendererClasses [&NiagaraMeshRendererProperties, &NiagaraSpriteRendererProperties]
  renderers [
   {rendererClass=&NiagaraMeshRendererProperties, rendererIndex=0}
   {rendererClass=&NiagaraMeshRendererProperties, rendererIndex=1}
   {rendererClass=&NiagaraMeshRendererProperties, rendererIndex=2}
   {rendererClass=&NiagaraMeshRendererProperties, rendererIndex=3}
   {rendererClass=&NiagaraSpriteRendererProperties, rendererIndex=4}
  ]
 }
}
summary {
 systemName=NS_Beat
 emitters [
  {emitterName=NE_VC_Boost_Ribbon_Parent, rendererClasses [&NiagaraRibbonRendererProperties]}
  {
   emitterName=NE_VC_Boost_Sprite_Parent007
   rendererClasses [&NiagaraMeshRendererProperties, &NiagaraSpriteRendererProperties]
  }
  {
   emitterName=NE_VC_Boost_Sprite_Parent010
   rendererClasses [&NiagaraMeshRendererProperties, &NiagaraSpriteRendererProperties]
  }
  {emitterName=NE_VC_Boost_Sprite_Parent, rendererClasses [&NiagaraSpriteRendererProperties]}
 ]
 userVariables [
  User.bIsCustomColor {type=NiagaraBool}
  User.bIsFrontendPreview {type=NiagaraBool}
  User.bIsPainted {type=NiagaraBool}
  User.ColorVariantID {type=NiagaraInt32}
  User.CustomColor {type=LinearColor, description="", defaultValue {struct=LinearColor, value=<rgba 5 0 4.8885 1>}}
  User.DTCarOffset {type=NiagaraPosition, description="", defaultValue {}}
  User.DTCarOffset {type=Vector3f, description="", defaultValue {struct=Vector3f, value=<0 0 0>}}
  User.DTCarOffset001 {type=Vector3f, description="", defaultValue {struct=Vector3f, value=<0 0 0>}}
  User.PaintedColor {type=LinearColor, description="", defaultValue {struct=LinearColor, value=<rgba 1 1 1 1>}}
  User.PaintedScalar {type=NiagaraFloat, defaultValue=1, description=""}
  User.PortCount {type=NiagaraInt32, defaultValue=2}
 ]
}
